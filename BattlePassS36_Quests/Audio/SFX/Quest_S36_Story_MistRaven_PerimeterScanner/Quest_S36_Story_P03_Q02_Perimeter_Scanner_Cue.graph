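# Sound cue for the perimeter scanner in a Season 36 MistRaven story quest, routed to the quests sound class.
# A mixer layers one of three equally weighted random scanner waves, boosted and passed through a fixed-pitch modulator, with a quieter reused reward-screen chest upgrade sound.
# It borrows a supply-drop ambient loop attenuation asset.
# A concurrency override allows one instance per owner and blocks new plays while it is active.
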SoundCue /BattlePassS36_Quests/Audio/SFX/Quest_S36_Story_MistRaven_PerimeterScanner/Quest_S36_Story_P03_Q02_Perimeter_Scanner_Cue
root=Mixer_0
nodes [
 {id=Mixer_0, children [Modulator_0,WavePlayer_4], properties {inputVolume [3,0.75]}}
 {id=Modulator_0, children [Random_0], properties {pitchMax=1, pitchMin=1}}
 {id=Random_0, children [WavePlayer_2,WavePlayer_3,WavePlayer_0], properties {weights [1,1,1]}}
 {
  id=WavePlayer_0
  properties {soundWaveAssetPtr=&/Quest_S36_Story_P03_Q02_Perimeter_Scanner_03}
 }
 {
  id=WavePlayer_2
  properties {soundWaveAssetPtr=&/Quest_S36_Story_P03_Q02_Perimeter_Scanner_01}
 }
 {
  id=WavePlayer_3
  properties {soundWaveAssetPtr=&/Quest_S36_Story_P03_Q02_Perimeter_Scanner_02}
 }
 {id=WavePlayer_4, properties {soundWaveAssetPtr=/Game/Sounds/Fort_UI_Sounds/Rewards/Fort_RewardScreen_TeamScore_ChestLootUpgrade}}
]
properties {
 attenuationSettings=/Game/Athena/Sounds/SupplyDrops/Fort_HiddenStar_Ambient_Loop_Att
 bOverrideConcurrency=true
 maxDistance=2500
 soundClassObject=&Fort_Quests
 concurrencyOverrides {bLimitToOwner=true, maxCount=1, resolutionRule=PreventNew}
}
